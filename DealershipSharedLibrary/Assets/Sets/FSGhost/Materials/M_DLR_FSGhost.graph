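# Translucent lit ghost material for the Dealership FSGhost set, usable on skeletal meshes and mesh particles.
# It builds a panning normal map, panning tinted emissive texture, and view-space cross-product matcap colored through the painted-variants function.
# An Erode function on a panning texture yields dissolve and edge masks, alongside an object-bounds gradient mask.
# Named reroutes feed contrast-adjusted opacity and an emissive blend with a boosted dissolve highlight color.
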Material /DealershipSharedLibrary/Assets/Sets/FSGhost/Materials/M_DLR_FSGhost
graph {
 properties {
  bAutomaticallySetUsageInEditor=false
  bUsedWithMeshDeformer=true
  bUsedWithMeshParticles=true
  bUsedWithNiagaraMeshParticles=true
  bUsedWithSkeletalMesh=true
  blendMode=BLEND_Translucent
  previewMesh=/DealershipExplosions/Assets/Explosion_Ghost/SK_Anim/SKM_Ghost_customGE_dab
  translucencyLightingMode=TLM_VolumetricDirectional
  usedShadingModels=Default Lit
 }
 nodes [
  Multiply(A=VectorParameter() {group=BaseColor, parameterName=BaseColor, sortPriority=1, defaultValue=<rgba 0 0 0 1>}.RGB, B=NamedRerouteUsage())
  n1 = TextureCoordinate()
  n2 = VectorParameter() {group=BaseColor, parameterName=BaseColor - UV, channelNames {a=PannerY, b=PannerX, g=ScaleY, r=ScaleX}, defaultValue=<rgba 1 1 0 0>}
  n3 = ComponentMask(None=@n2.RGB) {g=true, r=true}
  n4 = AppendVector(A=@n2.PannerX, B=@n2.PannerX)
  n5 = Time()
  n6 = TextureSampleParameter2D(UVs=Add(A=Multiply(A=@n1, B=@n3), B=Multiply(A=@n4, B=@n5))) {
   desc=Normal
   group=Normal
   parameterName=NormalTexture
   samplerType=SAMPLERTYPE_Normal
   texture=/DealershipSharedLibrary/Assets/Sets/Ghost/Textures/T_DLR_Ghost_N
  }
  NamedRerouteDeclaration(None=LinearInterpolate(
    A=Constant3Vector() {constant=<rgba 0 0 1 0>}
    B=@n6.RGB
    Alpha=ScalarParameter() {defaultValue=4, desc=Normal, group=Normal, parameterName=Normal - Intensity}
   )) {name=Normal, nodeColor=<rgba 0.3911 0.0385 1 1>}
  n7 = TextureCoordinate()
  n8 = VectorParameter() {group=Emissive, parameterName=Emissive - UV, channelNames {a=PannerY, b=PannerX, g=ScaleY, r=ScaleX}, defaultValue=<rgba 1 1 0 0>}
  n9 = ComponentMask(None=@n8.RGB) {g=true, r=true}
  n10 = AppendVector(A=@n8.PannerX, B=@n8.PannerX)
  n11 = Time()
  n12 = TextureSampleParameter2D(UVs=Add(A=Multiply(A=@n7, B=@n9), B=Multiply(A=@n10, B=@n11))) {group=Emissive, parameterName=EmissiveTexture, sortPriority=1, texture=/DealershipSharedLibrary/Assets/Sets/Ghost/Textures/T_DLR_Ghoul_C}
  n13 = ScalarParameter() {defaultValue=0.1, group=Emissive, parameterName=Texture Amount}
  NamedRerouteDeclaration(None=Multiply(
    A=VectorParameter() {group=Emissive, parameterName=Emissive - Color, defaultValue=<rgba 0.2195 1 0.807 1>}.RGB
    B=Multiply(A=@n12.RGB, B=@n13)
   )) {name=EmissiveColor, nodeColor=<rgba 1 0.3281 0 1>}
  n14 = TransformPosition(None=WorldPosition().XYZ) {transformSourceType=TRANSFORMPOSSOURCE_World, transformType=TRANSFORMPOSSOURCE_View}
  n15 = VertexNormalWS()
  n16 = MaterialFunctionCall(Color=CrossProduct(A=Normalize(VectorInput=@n14), B=Transform(None=@n15) {transformSourceType=TRANSFORMSOURCE_World, transformType=TRANSFORM_View})) {materialFunction=&SplitComponents}
  n17 = MaterialFunctionCall(R=@n16.G, G=@n16.G) {materialFunction=&AppendMany}
  n18 = Constant2Vector() {g=0.5, r=0.5}
  n19 = TextureSampleParameter2D(UVs=Add(A=Multiply(A=@n17.RG, B=@n18)) {constB=0.5}) {
   desc=Matcap
   group=Matcap
   parameterName=Matcap - Texture
   texture=/DealershipSharedLibrary/Assets/Sets/Ghost/Textures/T_DLR_LBankReflection_Matcap
  }
  n20 = Power(Base=@n19.RGB, Exp=ScalarParameter() {defaultValue=1, desc=Matcap, group=Matcap, parameterName=Matcap - Power})
  n21 = ScalarParameter() {defaultValue=1, group=Matcap, parameterName=Matcap - Multiply}
  n22 = VectorParameter() {desc=Matcap, group=Matcap, parameterName=Matcap - Color, defaultValue=<rgba 0.2195 1 0.807 1>}
  NamedRerouteDeclaration(None=Multiply(
    A=Multiply(A=@n20, B=@n21)
    B=MaterialFunctionCall!(Primary Color=@n22.RGBA, Tertiary Color, "Secondary  Color") {materialFunction=/DealershipSharedLibrary/FX/Materials/MaterialFunctions/MF_DLR_PaintedVariantsData}.Primary RGB
   )) {name=Matcap, nodeColor=<rgba 1 0 0.6328 1>}
  n23 = TextureCoordinate() {coordinateIndex=1}
  n24 = If(A=ScalarParameter() {group=Dissolve, parameterName=Dissolve - UV Channel}, "A > B"=@n23, "A == B"=@n23, "A < B"=TextureCoordinate())
  n25 = VectorParameter() {
   desc=Dissolve
   group=Dissolve
   parameterName=Dissolve - UV
   channelNames {a=PannerY, b=PannerX, g=ScaleY, r=ScaleX}
   defaultValue=<rgba 1 -1 0 0>
  }
  n26 = ComponentMask(None=@n25.RGB) {g=true, r=true}
  n27 = AppendVector(A=@n25.PannerX, B=@n25.PannerX)
  n28 = Time()
  n29 = TextureSampleParameter2D(UVs=Add(A=Multiply(A=@n24, B=@n26), B=Multiply(A=@n27, B=@n28))) {
   desc=Dissolve
   group=Dissolve
   parameterName=Dissolve - Texture
   texture=/DealershipSharedLibrary/Assets/Sets/Aquarium/Textures/T_Aquarium_DissolveUpDown
  }
  n30 = MaterialFunctionCall!(
   Erosion=ScalarParameter() {desc=Dissolve, group=Dissolve, parameterName=Dissolve}
   Mask=@n29.R
   Edge Width=ScalarParameter() {defaultValue=0.1, desc=Dissolve, group=Dissolve, parameterName=Dissolve Edge Width}
   Width=Constant() {r=0.25}
  ) {materialFunction=/DealershipSharedLibrary/FX/Materials/MaterialFunctions/MF_DLR_Erode}
  NamedRerouteDeclaration(None=@n30.Result) {name=Dissolve, nodeColor=<rgba 0 1 0.0859 1>}
  NamedRerouteDeclaration(None=@n30.Edge) {name=Dissolve Edge, nodeColor=<rgba 1 0 0.2344 1>}
  n31 = Divide(
   A=LocalPosition().XYZ
   B=Multiply(A=ObjectBounds(), B=VectorParameter() {group=Mask, parameterName=Mask - UV, defaultValue=<rgba 1 1 0 0>}.RGB)
  )
  n32 = ScalarParameter() {defaultValue=0.5, group=Mask, parameterName=Mask Y - Gradient Offset}
  n33 = Power(
   Base=Divide(A=Add(A=@n31, B=@n32), B=ScalarParameter() {group=Mask, parameterName=Mask Y - Gradient Height})
   Exp=ScalarParameter() {defaultValue=1, group=Mask, parameterName=Mask Y - Gradient Power}
  )
  n34 = ScalarParameter() {defaultValue=-1, group=Mask, parameterName=Mask Y - Gradient Bias}
  NamedRerouteDeclaration(None=Clamp(None=Add(A=@n33, B=@n34))) {name=Mask Y, nodeColor=<rgba 0 0.1562 1 1>}
  n35 = NamedRerouteUsage()
  n36 = NamedRerouteUsage()
  NamedRerouteDeclaration(None=MaterialFunctionCall(In=Multiply(A=@n35, B=@n36), Contrast=ScalarParameter() {group=Dissolve, parameterName=Dissolve Edge Contrast}) {materialFunction=&CheapContrast}.Result) {name=Opacity, nodeColor=<rgba 1 0 0.3984 1>}
  n37 = VectorParameter() {group=Dissolve, parameterName=Emissive - Dissolve Highlight Color, defaultValue=<rgba 0.2195 1 0.807 1>}
  n38 = Multiply(A=@n37.RGB, B=ScalarParameter() {defaultValue=50, group=Dissolve, parameterName=Dissolve Highlight Gain})
  n39 = Add(A=NamedRerouteUsage(), B=NamedRerouteUsage())
  n40 = NamedRerouteUsage()
  NamedRerouteDeclaration(None=MaterialFunctionCall(Emissive=LinearInterpolate(A=@n38, B=@n39, Alpha=@n40)) {materialFunction=&MF_FX_EmissiveAdjustments}.Result) {name=Emissive Result, nodeColor=<rgba 1 0 0.9609 1>}
  NamedRerouteUsage()
  NamedRerouteUsage()
  NamedRerouteUsage()
 ]
}
